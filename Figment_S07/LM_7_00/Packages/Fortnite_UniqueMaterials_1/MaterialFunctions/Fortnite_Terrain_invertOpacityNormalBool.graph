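# Material function for Fortnite terrain materials that rebuilds a material attribute set with Opacity inverted.
# A Use Spec static switch keeps Metallic, Specular and Roughness or substitutes non-metallic, zero-specular, fully rough constants.
# A Use Normals switch keeps the input Normal or substitutes the world-space vertex normal; both switches default to true.
# Most other rebuilt channels, including Emissive, OpacityMask and WorldPositionOffset, are wired from the input's BaseColor.
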MaterialFunction /Figment_S07/LM_7_00/Packages/Fortnite_UniqueMaterials_1/MaterialFunctions/Fortnite_Terrain_invertOpacityNormalBool
graph {nodes [
  n1 = BreakMaterialAttributes!(Attr=FunctionInput(Preview=MakeMaterialAttributes!(
     BaseColor
     Metallic
     Specular
     Roughness
     Anisotropy
     EmissiveColor
     Opacity
     OpacityMask
     Normal
     Tangent
     WorldPositionOffset
     SubsurfaceColor
     ClearCoat
     ClearCoatRoughness
     AmbientOcclusion
     Refraction
     CustomizedUVs_0
     CustomizedUVs_1
     CustomizedUVs_2
     CustomizedUVs_3
     CustomizedUVs_4
     CustomizedUVs_5
     CustomizedUVs_6
     CustomizedUVs_7
     PixelDepthOffset
     ShadingModel
     Displacement
    )) {inputName=Mat attributes, inputType=FunctionInput_MaterialAttributes})
  n2 = Constant()
  n3 = FunctionInput(Preview=StaticBool() {value=true}) {bUsePreviewValueAsDefault=true, inputName=Use Spec, inputType=FunctionInput_StaticBool, sortPriority=1}
  n4 = Constant() {r=1}
  n5 = VertexNormalWS()
  n6 = FunctionInput(Preview=StaticBool() {value=true}) {bUsePreviewValueAsDefault=true, inputName=Use Normals, inputType=FunctionInput_StaticBool, sortPriority=1}
  FunctionOutput(None=MakeMaterialAttributes(
    BaseColor=@n1.BaseColor
    Metallic=StaticSwitch(True=@n1.Metallic, False=@n2, Value=@n3)
    Specular=StaticSwitch(True=@n1.Specular, False=@n2, Value=@n3)
    Roughness=StaticSwitch(True=@n1.Roughness, False=@n4, Value=@n3)
    EmissiveColor=@n1.BaseColor
    Opacity=OneMinus(None=@n1.Opacity)
    OpacityMask=@n1.BaseColor
    Normal=StaticSwitch(True=@n1.Normal, False=@n5, Value=@n6)
    WorldPositionOffset=@n1.BaseColor
    SubsurfaceColor=@n1.BaseColor
    ClearCoat=@n1.BaseColor
    ClearCoatRoughness=@n1.BaseColor
    AmbientOcclusion=@n1.BaseColor
    Refraction=@n1.BaseColor
    CustomizedUVs_0=@n1.BaseColor
    CustomizedUVs_1=@n1.BaseColor
    CustomizedUVs_2=@n1.BaseColor
    CustomizedUVs_3=@n1.BaseColor
    CustomizedUVs_4=@n1.BaseColor
    CustomizedUVs_5=@n1.BaseColor
    CustomizedUVs_6=@n1.BaseColor
    CustomizedUVs_7=@n1.BaseColor
    PixelDepthOffset=@n1.BaseColor
   ))
 ]}
