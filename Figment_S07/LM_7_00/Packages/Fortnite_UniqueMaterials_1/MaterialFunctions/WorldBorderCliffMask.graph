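# Material function that builds a height mask from world Z, named for the cliffs at the world border.
# It extracts the Z component of world position, adds the "World Z mask distance" scalar parameter, and clamps the result.
# With no division or scaling before the clamp, the mask is a near-hard cutoff at the negative of that distance rather than a soft gradient.
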MaterialFunction /Figment_S07/LM_7_00/Packages/Fortnite_UniqueMaterials_1/MaterialFunctions/WorldBorderCliffMask
graph {nodes [
  n1 = MaterialFunctionCall(Float3=WorldPosition().XYZ) {materialFunction=&BreakOutFloat3Components}
  n2 = ScalarParameter() {defaultValue=3000, parameterName=World Z mask distance}
  FunctionOutput(None=Clamp(None=Add(A=@n1.B, B=@n2) {constB=1575}))
 ]}
